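# PistolCanary_tac1200_tec is an empty Level Sequence in the PistolCanary sequence folder for tac1200.
# It has no bindings and no tracks, so it animates nothing as it stands.
# It is set to a 30 fps display rate on the 24000 tick resolution.
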LevelSequence /PistolCanary/Sequence/tac/tac1200/PistolCanary_tac1200_tec
bindings []
display_rate {denominator=1, numerator=30}
tick_resolution {denominator=1, numerator=24000}
tracks []
